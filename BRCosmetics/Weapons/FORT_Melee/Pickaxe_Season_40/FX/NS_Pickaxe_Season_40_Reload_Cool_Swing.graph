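# Niagara swing effect for the Season 40 "Reload Cool" pickaxe, with one sprite emitter (Drops) emitting short-lived droplets.
# A variant selector scales spawn rate, with zero for the original variant.
# Particles spawn in a small sphere around sampled sockets and inherit capped velocity, then drag, spin, grow and fade while shifting from orange through cyan to blue.
# Per-particle size factors normalize dynamic material parameters, alongside random values.
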NiagaraSystem /BRCosmetics/Weapons/FORT_Melee/Pickaxe_Season_40/FX/NS_Pickaxe_Season_40_Reload_Cool_Swing
dependencies {
 usedDataInterfaces [&NiagaraDataInterfaceCurve, &NiagaraDataInterfaceColorCurve, &NiagaraDataInterfaceSocketReader]
 usedDynamicInputs []
 usedModules [
  &self:SystemSpawnScript.SystemScriptSource.SystemScriptGraph.NiagaraNodeAssignment_1.SetVariables_#1
  &FN_VariantManager
  &SystemState
 ]
}
emitters {Drops {
  emitterUpdateScript [
   &EmitterState()
   &SpawnRate(
    SpawnRate=&Multiply_Float(
     A=&RandomRangeFloat(Minimum=60, Maximum=80)
     B=/BRCosmetics/Weapons/FORT_Melee/Pickaxe_GateHound/FX/VariantManager/VariantFloatSelector(
      Original Float:NiagaraFloat=0
      Variant 1 Float:NiagaraFloat=1
      Variant 2 Float:NiagaraFloat=?
      Variant 3 Float:NiagaraFloat=?
      Variant Amount:NiagaraInt32=0
     )
    )
    Use Spawn Probability=false
   ) {module=SpawnRate001}
  ]
  particleSpawnScript [
   &InitializeParticle(
    Lifetime Max=0.35
    Lifetime Min=0.2
    Lifetime Mode=::NewEnumerator1
    Position Offset=<0 0 100>
    Sprite Rotation Mode=::NewEnumerator2
    Sprite Size=&Multiply_Vector2D_ByFloat(Float=10, Vector2D=&RandomRangeVector2D(Minimum=<0.35 0.35>, Maximum=<1.5 1.5>))
    Sprite Size Mode=::NewEnumerator4
    UsePositionOffset=false
    Write Material Random=true
   )
   &SocketLocation(
    Interpolate Socket Sampling=true
    SamplePosOffset=true
    Sampled Position Offset=<0 0 60>
    Socket Reader=$System.Initial.Socket Reader
    Socket Sampling Mode=::NewEnumerator3
    VelocitySpeedLimit=false
   )
   &ShapeLocation(MaskSpawnGroup=false, Override Seed=false, Sphere Radius=7, UseOwnerScale=true)
   &InheritVelocity(Inherited Velocity Speed Limit=25, Speed Limit=true, UseSpeedThreshold=false)
   /SetVariables(Particles.SizeFactor:Vector2f=/Niagara/DynamicInputs/Divide/Divide_Vector2(A:Vector2f=<10 10>, B:Vector2f=$Particles.SpriteSize)) {isSetParameters=true}
  ]
  particleUpdateScript [
   &ParticleState()
   &ScaleSpriteSize(Uniform Curve Sprite Scale=DataInterface() {propertyValues {Curve {
       defaultValue=0
       keys [{interpMode=RCIM_Cubic, tangentMode=RCTM_User, time=0, value=0.5}, {interpMode=RCIM_Cubic, tangentMode=RCTM_User, time=1, value=1}]
      }}})
   &SpriteRotationRate(Rotation Rate=&RandomRangeFloat(Minimum=-40, Maximum=40))
   &Drag(Drag=2, Use Linear Drag=true, Use Rotational Drag=true)
   &SolveForcesAndVelocity(Clamp Velocity=false, Limit Acceleration=false)
   &ScaleColor(
    Scale Alpha=&FloatFromCurve(FloatCurve=DataInterface() {propertyValues {Curve {keys [
         <at 0 0 RCIM_Cubic>
         {arriveTangentWeight=0.0581, interpMode=RCIM_Cubic, leaveTangentWeight=0.3488, tangentMode=RCTM_User, time=0.2, value=1}
         {arriveTangentWeight=0.299, interpMode=RCIM_Cubic, leaveTangentWeight=0.1495, tangentMode=RCTM_User, time=0.7, value=1}
         <at 1 0>
        ]}}})
    Scale RGB=&MakeVectorFromLinearColorRGB(Color=&ColorFromCurve(ColorCurve=DataInterface() {propertyValues {
        AlphaCurve {keys [<at 0 1 RCIM_Cubic>,<at 1 0 RCIM_Cubic>]}
        BlueCurve {keys [<at 0 0 RCIM_Cubic>,<at 0.4275 0>,<at 0.8042 0.9531>,<at 1 1 RCIM_Cubic>]}
        GreenCurve {keys [<at 0 1.25 RCIM_Cubic>,<at 0.4275 0.5>,<at 0.8042 0.7836>,<at 1 0.0348 RCIM_Cubic>]}
        RedCurve {keys [<at 0 2.5 RCIM_Cubic>,<at 0.4275 1.5>,<at 0.8042 0.0248>,<at 1 0 RCIM_Cubic>]}
       }}))
    ScaleA=true
    ScaleRGB=true
   ) {module=ScaleColor001}
   &DynamicMaterialParameters(
    Index 1 Param 3=&RandomRangeFloat(Minimum=0.2, Maximum=0.65)
    Index 2 Param 1=&Divide_Float(A=0.25, B=&MakeFloatFromVector2d(Vector 2D=$Particles.SizeFactor))
    Index 2 Param 2=&Divide_Float(A=0.1, B=&MakeFloatFromVector2d(Vector 2D=$Particles.SizeFactor, Channel=::NewEnumerator1))
    Index 2 Param 3=&RandomRangeFloat(Minimum=0, Maximum=1)
    Index 2 Param 4=&RandomRangeFloat(Minimum=0, Maximum=1)
    Module.Index1Float0Write=false
    Module.Index1Float1Write=false
    Module.Index1Float2Write=true
    Module.Index1Float3Write=false
    Module.Index2Float0Write=true
    Module.Index2Float1Write=true
    Module.Index2Float2Write=true
    Module.Index2Float3Write=true
    Write Parameter Index 0=::NewEnumerator0
    Write Parameter Index 1=::NewEnumerator1
    Write Parameter Index 2=::NewEnumerator1
   )
  ]
  rendererClasses [&NiagaraSpriteRendererProperties]
  renderers [{rendererClass=&NiagaraSpriteRendererProperties}]
 }}
summary {
 systemName=NS_Pickaxe_Season_40_Reload_Cool_Swing
 emitters [{emitterName=Drops, rendererClasses [&NiagaraSpriteRendererProperties]}]
 userVariables [
  User.bIsFrontEnd {type=NiagaraBool}
  User.bIsFrontEndPreview {type=NiagaraBool}
  User.Variant {type=NiagaraFloat, defaultValue=0, description=""}
 ]
}
