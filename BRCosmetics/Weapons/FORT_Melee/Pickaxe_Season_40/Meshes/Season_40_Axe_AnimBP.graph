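# Animation Blueprint for the Season 40 pickaxe that adds secondary physics motion to the axe's handle bone.
# An Anim Dynamics node runs on dyn_handle_00 over the input pose, using an angular spring, angular limits and bone-relative simulation.
# Each update stores gravity from a parameter-asset function and a front-end-dependent Alpha FE, but the node's gravity and alpha inputs show constants rather than those variables.
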AnimBlueprint /BRCosmetics/Weapons/FORT_Melee/Pickaxe_Season_40/Meshes/Season_40_Axe_AnimBP
target_skeleton=&/Season_40_Axe_Skeleton
graphs [
 AnimGraph {nodes [
   AnimGraphNode_SubInput_0 = Input Pose()
   n1 = Anim Dynamics - Bone: dyn_handle_00!(ComponentPose=Local To Component(LocalPose=@AnimGraphNode_SubInput_0), Alpha=1, GravityOverride:Vector="0,0,-980", ExternalForce="0,0,0") -> (Pose) {
    angularDampingOverride=1
    angularSpringConstant=1900
    bAngularSpring=true
    bOverrideAngularDamping=true
    bUseGravityOverride=true
    gravityScale=0.35
    simulationSpace=BoneRelative
    boundBone {boneName=dyn_handle_00}
    componentAppliedLinearAccClamp=<5000 5000 5000>
    componentLinearAccScale=<0.8 0.8 0.8>
    componentLinearVelScale=<0.8 0.8 0.8>
    gravityOverride {z=-980}
    physicsBodyDefinitions [{
      boundBone {boneName=dyn_handle_00}
      boxExtents=<15 15 15>
      constraintSetup {angularLimitsMax=<0 0 15>, angularLimitsMin=<0 0 -70>, angularTarget=<1 0 0>}
      localJointOffset=<1 0 0>
     }]
    simSpaceSettings {maxAngularAcceleration=35, maxAngularVelocity=7, simSpaceAngularAlpha=1}
   }
   Output Pose(Result=Component To Local(ComponentPose=@n1)) {layerGroup=None}
  ]}
 EventGraph {nodes [
   Event BlueprintUpdateAnimation()
   > n1 = Sequence()
   > n2 = "Gravity Override State Function (Parameter Asset) (Nativized)"(
    CharacterPartAnimInstance=Self-Reference()
    GameTimeInSeconds=0
    Parameters=&/Fortnite_Season_40_Axe_Parameters
   )
   > Set Gravity Override(Gravity Override=@n2)
   n3 = Branch(execute=@n1.then_1, Condition=Get bIsInFrontEnd())
   > Set Alpha FE(Alpha FE:float=0.75) -> (Output_Get:float)
   Set Alpha FE(execute=@n3.else, Alpha FE=1)
   TryGetPawnOwner()
  ]}
]
variables [
 Alpha FE {type=number}
 Gravity Override {type=Vector}
]
